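# Sound cue for the Flopping Rabbit fail sound, filed under the Flopper item gameplay cue sounds.
# A single wave player feeds the root directly and plays the Target_Fail wave, with no randomization or modulation.
# It uses the Fort_Consumables sound class and the in-world forged consumable attenuation asset.
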SoundCue /FNGameplayCues/Athena/Sounds/Items/Flopper/Flopping_Rabbit_Fail_Cue
root=WavePlayer_1
nodes [{id=WavePlayer_1, properties {soundWaveAssetPtr=&/Target_Fail}}]
properties {
 attenuationSettings=/Game/Sounds/Attenuation/Fort_ForgedConsumable_InWorld_Att
 maxDistance=1000
 soundClassObject=&Fort_Consumables
}
