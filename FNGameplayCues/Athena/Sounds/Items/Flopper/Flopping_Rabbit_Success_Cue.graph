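# Sound cue for the Flopper item's "Flopping Rabbit" success sound, stored with Athena gameplay cue sounds.
# A random node picks with equal weight between two Target_Success wave variants.
# It uses a forged-consumable in-world attenuation but is routed to the critical in-game UI sound class, with boosted volume.
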SoundCue /FNGameplayCues/Athena/Sounds/Items/Flopper/Flopping_Rabbit_Success_Cue
root=Random_0
nodes [
 {id=Random_0, children [WavePlayer_0,WavePlayer_1], properties {weights [1,1]}}
 {id=WavePlayer_0, properties {soundWaveAssetPtr=&/Target_Success}}
 {id=WavePlayer_1, properties {soundWaveAssetPtr=&/Target_Success_02}}
]
properties {
 attenuationSettings=/Game/Sounds/Attenuation/Fort_ForgedConsumable_InWorld_Att
 maxDistance=1000
 soundClassObject=&Fort_UI_InGame_Critical
 volumeMultiplier=1.75
}
