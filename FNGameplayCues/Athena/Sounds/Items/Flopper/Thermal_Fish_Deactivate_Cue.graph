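# Sound cue for the Thermal Fish deactivation sound, among the Flopper item sounds, in the consumables sound class.
# A mixer layers two Thermal_Fish_Deactivate wave players: one delayed by a fixed quarter second, the other immediate and slightly quieter.
# The immediate layer passes through a modulator that halves its pitch, adding a low undertone beneath the delayed, unshifted sound.
# Spatial falloff uses the shared forged-consumable in-world attenuation asset.
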SoundCue /FNGameplayCues/Athena/Sounds/Items/Flopper/Thermal_Fish_Deactivate_Cue
root=Mixer_0
nodes [
 {id=Delay_0, children [WavePlayer_0], properties {delayMax=0.25, delayMin=0.25}}
 {id=Mixer_0, children [Delay_0,Modulator_0], properties {inputVolume [1,0.7]}}
 {id=Modulator_0, children [WavePlayer_1], properties {pitchMax=0.5, pitchMin=0.5, volumeMax=1, volumeMin=1}}
 {id=WavePlayer_0, properties {soundWaveAssetPtr=&/Thermal_Fish_Deactivate}}
 {id=WavePlayer_1, properties {soundWaveAssetPtr=&Thermal_Fish_Deactivate}}
]
properties {
 attenuationSettings=/Game/Sounds/Attenuation/Fort_ForgedConsumable_InWorld_Att
 maxDistance=1000
 soundClassObject=&Fort_Consumables
}
