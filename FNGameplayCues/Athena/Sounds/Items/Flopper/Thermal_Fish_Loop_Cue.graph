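# Looping sound cue for the Thermal Fish, a Flopper item, routed to the consumables sound class with in-world consumable attenuation.
# A looping wave player feeds an Enveloper that quickly fades volume in, holds it briefly, then eases it down to about half over several seconds.
# The pitch curve has a single key, so the envelope shapes only volume over time.
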SoundCue /FNGameplayCues/Athena/Sounds/Items/Flopper/Thermal_Fish_Loop_Cue
root=Enveloper_0
nodes [
 {
  id=Enveloper_0
  children [WavePlayer_1]
  properties {pitchCurve {editorCurveData {keys [{time=0}]}}, volumeCurve {editorCurveData {keys [<at 0 0.01>,<at 0.25 1>,<at 2 1>,<at 10 0.4963>]}}}
 }
 {id=WavePlayer_1, properties {bLooping=true, soundWaveAssetPtr=&/Thermal_Fish_Loop}}
]
properties {
 attenuationSettings=/Game/Sounds/Attenuation/Fort_ForgedConsumable_InWorld_Att
 maxDistance=1000
 soundClassObject=&Fort_Consumables
 volumeMultiplier=0.45
}
