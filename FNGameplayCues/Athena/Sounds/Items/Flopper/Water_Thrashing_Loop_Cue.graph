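# Looping sound cue for Flopper water thrashing, under the Athena item sounds.
# A single looping wave player feeds the root with no randomization or mixing, using slight pitch and volume boosts.
# It uses the Fort_Loot_Collect sound class and a shared FloppingRabbitMinigame loop attenuation asset.
# Attenuation override values are stored, but the override flag is not enabled.
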SoundCue /FNGameplayCues/Athena/Sounds/Items/Flopper/Water_Thrashing_Loop_Cue
root=WavePlayer_1
nodes [{id=WavePlayer_1, properties {bLooping=true, soundWaveAssetPtr=&/Water_Thrashing_Loop}}]
properties {
 attenuationSettings=/FNGameplayCues/Sounds/Attenuation/Fort_FloppingRabbitMinigame_Loop_Att
 maxDistance=8000
 pitchMultiplier=1.02
 soundClassObject=&Fort_Loot_Collect
 volumeMultiplier=1.35
 attenuationOverrides {falloffDistance=6000, attenuationShapeExtents {x=2500}}
}
